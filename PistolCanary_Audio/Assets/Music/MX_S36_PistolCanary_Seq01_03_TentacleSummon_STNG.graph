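# MetaSound source preset that plays the TentacleSummon music stinger for sequence 01_03 of the PistolCanary event as a stereo one-shot.
# It inherits its graph from the PreRitual stinger MetaSound, supplying its own wave asset and exposing master volume, fade-out trigger and encounter pause inputs.
# Playback routes through the event music sound class with the PistolCanary POI music attenuation, so it is spatialized around that location.
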MetaSoundSource /PistolCanary_Audio/Assets/Music/MX_S36_PistolCanary_Seq01_03_TentacleSummon_STNG
preset_parent=&/MX_S36_PistolCanary_Seq01_01_PreRitual_STNG
inputs [
 FadeOut {type=Trigger}
 MasterVolume {type=Float}
 Music.WaveAsset {type=WaveAsset:Array}
 MusicEncounter.bIsPaused {type=Bool}
 MusicEncounter.PauseFadeOutTime {type=Time}
 UE.Source.OnPlay {type=Trigger}
]
interfaces [UE.OutputFormat.Stereo:1.0, UE.Source.OneShot:1.0, UE.Source:1.1]
outputs [
 "UE.OutputFormat.Stereo.Audio:0" {type=Audio}
 "UE.OutputFormat.Stereo.Audio:1" {type=Audio}
 UE.Source.OneShot.OnFinished {type=Trigger}
]
pages [Default {input_defaults {
   FadeOut=null
   MasterVolume=0.5
   MusicEncounter.PauseFadeOutTime=3
   MusicEncounter.bIsPaused=null
   UE.Source.OnPlay=null
   Music.WaveAsset [/PistolCanary_Audio/Assets/Music/Wav/MX_S36_PistolCanary_Seq1_03_TentacleSummon_STNG_V2]
  }}]
properties {
 assetImportData=&self:AssetImportData
 attenuationSettings=/PistolCanary_Audio/Assets/Settings/ATT_PistolCanary_POI_Music
 maxDistance=59960.5625
 numChannels=2
 outputFormat=Stereo
 sampleRate=48000
 soundClassObject=/Game/Sounds/FortSoundClassesAndModes/Fort_Events_Music
 volume=0.9
}
