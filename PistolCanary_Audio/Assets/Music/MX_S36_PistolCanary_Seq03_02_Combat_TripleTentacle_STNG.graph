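# One-shot stereo MetaSound source for the PistolCanary Seq03_02 "TripleTentacle" combat music stinger, routed to the event music sound class with POI music attenuation.
# On play, a random-get over the single-entry wave array feeds a stereo wave player, scaled by MasterVolume and an interpolated fade gain.
# FadeOut or pausing ramps gain to zero over PauseFadeOutTime and signals finished after that delay.
# Unpausing restores gain and resets the delay; natural wave completion also finishes.
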MetaSoundSource /PistolCanary_Audio/Assets/Music/MX_S36_PistolCanary_Seq03_02_Combat_TripleTentacle_STNG
inputs [
 FadeOut {type=Trigger}
 MasterVolume {type=Float}
 Music.WaveAsset {type=WaveAsset:Array}
 MusicEncounter.bIsPaused {type=Bool}
 MusicEncounter.PauseFadeOutTime {type=Time}
 UE.Source.OnPlay {type=Trigger}
]
interfaces [UE.OutputFormat.Stereo:1.0, UE.Source.OneShot:1.0, UE.Source:1.1]
outputs [
 "UE.OutputFormat.Stereo.Audio:0" {type=Audio}
 "UE.OutputFormat.Stereo.Audio:1" {type=Audio}
 UE.Source.OneShot.OnFinished {type=Trigger}
]
pages [Default {
  input_defaults {
   FadeOut=null
   MasterVolume=1
   MusicEncounter.PauseFadeOutTime=3
   MusicEncounter.bIsPaused=null
   UE.Source.OnPlay=null
   Music.WaveAsset [/PistolCanary_Audio/Assets/Music/Wav/MX_S36_PistolCanary_Seq03_02_Combat_TripleTentacle_STNG]
  }
  nodes [
   TriggerOnValueChange_0 = UE.Trigger On Value Change:Bool(Value=^MusicEncounter.bIsPaused)
   n1 = TriggerCompare.Clamp:Bool(A=^MusicEncounter.bIsPaused, Compare=@TriggerOnValueChange_0.Trigger) {also {defaults {B="<MetaSoundBuilderResult.SUCCEEDED: 0>"}}}
   TriggerAnyTriggerAny2_1 = "TriggerAny.Trigger Any (2)"(In 0=^FadeOut, In 1=@n1.True)
   n2 = Value.Value:Float(Set=@TriggerAnyTriggerAny2_1.Out, Reset=@n1.False) {also {defaults {Init Value=1, Target Value=0}}}
   n3 = UE.Multiply:Float(
    PrimaryOperand=^MasterVolume
    AdditionalOperands=UE.InterpTo:Audio(Interp Time=^MusicEncounter.PauseFadeOutTime, Target=@n2.Output Value).Value
   )
   ArrayRandomGet_0 = Array.Random Get:WaveAsset:Array(
    In Array=^Music.WaveAsset
    Next=^UE.Source.OnPlay
    No Repeats=UE.Divide:Int32(PrimaryOperand=Array.Num:WaveAsset:Array(Array=^Music.WaveAsset).Num).Out
   ) {also {defaults {Enable Shared State="<MetaSoundBuilderResult.SUCCEEDED: 0>", Shared State Behavior=2}}}
   WavePlayer_0 = UE.Wave Player:Stereo(Play=@ArrayRandomGet_0.On Next, Wave Asset=@ArrayRandomGet_0.Value)
   Multiply_0 = UE.Multiply:Audio by Float(AdditionalOperands=@n3.Out, PrimaryOperand=@WavePlayer_0.Out Left)
   Multiply_1 = UE.Multiply:Audio by Float(AdditionalOperands=@n3.Out, PrimaryOperand=@WavePlayer_0.Out Right)
   TriggerDelay_0 = UE.Trigger Delay(Delay Time=^MusicEncounter.PauseFadeOutTime, Reset=@n2.On Reset, In=@n2.On Set)
   TriggerAnyTriggerAny2_0 = "TriggerAny.Trigger Any (2)"(In 1=@TriggerDelay_0.Out, In 0=@WavePlayer_0.On Finished)
  ]
  links [
   {dst="@UE.OutputFormat.Stereo.Audio:0", src=Multiply_0.Out}
   {dst="@UE.OutputFormat.Stereo.Audio:1", src=Multiply_1.Out}
   {dst="@UE.Source.OneShot.OnFinished", src=TriggerAnyTriggerAny2_0.Out}
  ]
 }]
properties {
 assetImportData=&self:AssetImportData
 attenuationSettings=/PistolCanary_Audio/Assets/Settings/ATT_PistolCanary_POI_Music
 maxDistance=59960.5625
 numChannels=2
 outputFormat=Stereo
 sampleRate=48000
 soundClassObject=/Game/Sounds/FortSoundClassesAndModes/Fort_Events_Music
 volume=1.2
}
